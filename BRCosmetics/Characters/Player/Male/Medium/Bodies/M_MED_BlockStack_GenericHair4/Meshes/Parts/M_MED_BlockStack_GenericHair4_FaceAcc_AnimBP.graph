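# Animation Blueprint for the face accessory part of the male medium BlockStack GenericHair4 character, adding secondary motion to two front hair bones.
# Copies the pose from the mesh, set to use the attached parent and with a Head component input.
# Two chained Anim Dynamics nodes then simulate Hair_Front_01 and 02 with identical root-relative settings: angular springs, angular limits, reduced gravity and damped component motion.
# The event graph holds only an unconnected update event.
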AnimBlueprint /BRCosmetics/Characters/Player/Male/Medium/Bodies/M_MED_BlockStack_GenericHair4/Meshes/Parts/M_MED_BlockStack_GenericHair4_FaceAcc_AnimBP
target_skeleton=&/M_MED_BlockStack_GenericHair4_FaceAcc_Skeleton
graphs [
 AnimGraph {nodes [
   n1 = Get Head() -> (Head:Skeletal Mesh Component Budgeted Object Reference)
   n2 = Anim Dynamics - Bone: Hair_Front_01_dyn(
    ComponentPose=Local To Component(LocalPose=Copy Pose From Mesh(SourceMeshComponent=@n1) {bUseAttachedParent=true})
    Alpha=1
    ExternalForce="0,0,0"
   ) {
    angularDampingOverride=1
    angularSpringConstant=10000
    bAngularSpring=true
    bOverrideAngularDamping=true
    bUseGravityOverride=true
    gravityScale=0.35
    simulationSpace=RootRelative
    boundBone {boneName=Hair_Front_01_dyn}
    componentAppliedLinearAccClamp=<4000 4000 4000>
    componentLinearAccScale=<0.5 0.5 0.6>
    componentLinearVelScale=<0.5 0.5 0.6>
    gravityOverride {z=-980}
    physicsBodyDefinitions [{
      boundBone {boneName=Hair_Front_01_dyn}
      boxExtents=<20 20 20>
      constraintSetup {angularLimitsMax=<20 25 8>, angularLimitsMin=<-20 -25 -8>, angularTarget=<1 0 0>}
      localJointOffset=<4 0 0>
     }]
    simSpaceSettings {maxAngularAcceleration=100, maxAngularVelocity=10, simSpaceAngularAlpha=1}
   }
   n3 = Anim Dynamics - Bone: Hair_Front_02_dyn(ComponentPose=@n2, Alpha=1, ExternalForce="0,0,0") {
    angularDampingOverride=1
    angularSpringConstant=10000
    bAngularSpring=true
    bOverrideAngularDamping=true
    bUseGravityOverride=true
    gravityScale=0.35
    simulationSpace=RootRelative
    boundBone {boneName=Hair_Front_02_dyn}
    componentAppliedLinearAccClamp=<4000 4000 4000>
    componentLinearAccScale=<0.5 0.5 0.6>
    componentLinearVelScale=<0.5 0.5 0.6>
    gravityOverride {z=-980}
    physicsBodyDefinitions [{
      boundBone {boneName=Hair_Front_02_dyn}
      boxExtents=<20 20 20>
      constraintSetup {angularLimitsMax=<20 25 8>, angularLimitsMin=<-20 -25 -8>, angularTarget=<1 0 0>}
      localJointOffset=<4 0 0>
     }]
    simSpaceSettings {maxAngularAcceleration=100, maxAngularVelocity=10, simSpaceAngularAlpha=1}
   }
   Output Pose(Result=Component To Local(ComponentPose=@n3)) {layerGroup=None}
  ]}
 EventGraph {nodes [Event BlueprintUpdateAnimation()]}
]
variables [Head {type=SkeletalMeshComponentBudgeted}]
